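# PA_LavaRex_Gadget is the physics asset for the LavaRex gadget prop mesh, filed under the cosmetic companions' emote props.
# It holds a single body on the Root bone, made of one tilted capsule slightly offset from the bone.
# It has no constraints, so it provides one simple collision shape rather than an articulated ragdoll.
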PhysicsAsset /CosmeticCompanions/Assets/Props/Emotes/LavaRex/Meshes/PA_LavaRex_Gadget
bodies {Root [{
   shapeType=Capsule
   length=71.1738
   radius=66.6414
   shapeName=Root_capsule
   center=<0.8553 9.2209 10.0435>
   rotation=<rot 18.2743 -22.5329 -90>
  }]}
constraints []
